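# Physics asset for the Modern Military glider skeletal mesh, a VinderTech glider chute cosmetic gadget.
# It holds a single box collision body on the Frame bone, offset and rotated to fit the glider.
# There are no constraints, so no other bones get physics bodies or jointed simulation.
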PhysicsAsset /BRCosmetics/Gadgets/Assets/VinderTech_GliderChute/Glider_Modern_Military_3/Meshes/SK_Glider_ModernMilitary_Physics
bodies {Frame [{shapeType=Box, extentX=185.4587, extentY=73.565, extentZ=377.5859, center=<-0.0034 -19.6102 -9.8464>, rotation=<rot 0 90.0001 -89.9998>}]}
constraints []
